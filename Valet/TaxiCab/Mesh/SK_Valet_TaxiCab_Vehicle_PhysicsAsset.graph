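# Physics asset for the taxi cab vehicle skeletal mesh in the Valet content.
# It holds a single body on the root bone, with no collision shapes listed for it.
# There are no constraints, so the asset defines no joints between bodies.
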PhysicsAsset /Valet/TaxiCab/Mesh/SK_Valet_TaxiCab_Vehicle_PhysicsAsset
bodies {root []}
constraints []
