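; Taxi cab variant of a Valet vehicle, extending the parent's cosmetics to cover its extra Taxi_Attachments mesh.
; On explosion it swaps the attachment's first material slot to a destroyed material, and on dissolve it drives that mesh's Fade parameter, calling the parent in both cases.
; It reports the attachment through getAttachmentMeshComponents and overrides PickRandomMaterial with an empty body, so no random material is picked.
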(event EventPickRandomMaterial)

(event EventExplodeCosmetics
  (params (DestroyedActor :type Actor%20Object%20Reference))
  ;[1 1 1] Set Material and fade out attachment
  (Rendering|Material|SetMaterial (Variables|Default|GetTaxi_Attachments) 0 "/Valet/Materials/M_FN_Valet_Destroyed_Master_InstAdjusted.M_FN_Valet_Destroyed_Master_InstAdjusted")
  (|CallParentFunction DestroyedActor))

(event EventDisolveVehicle
  (params (Fade :type Float%20%28double-precision%29))
  ;[1 1 1] Set Material and fade out attachment
  (Rendering|Material|SetScalarParameterValueonMaterials (Variables|Default|GetTaxi_Attachments) "Fade" Fade)
  (|CallParentFunction Fade))

(fn getAttachmentMeshComponents
  (return (Utilities|Array|MakeArray (Variables|Default|GetTaxi_Attachments))))

(fn ConstructionScript
  (|CallParentFunction))
